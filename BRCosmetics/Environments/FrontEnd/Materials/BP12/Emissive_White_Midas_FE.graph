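# Default Lit material for the BP12 front-end environment, named as a white emissive Midas variant.
# It holds three white or grey color constants and two vector parameters, one an over-bright HDR blue and the other a mid grey.
# No links between these nodes or to the material outputs appear in the graph.
# Usage is enabled for Nanite, skeletal meshes and mesh deformers.
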Material /BRCosmetics/Environments/FrontEnd/Materials/BP12/Emissive_White_Midas_FE
graph {
 properties {
  bUsedWithMeshDeformer=true
  bUsedWithNanite=true
  bUsedWithSkeletalMesh=true
  pixelDepthOffsetMode=PDOM_Legacy
  usedShadingModels=Default Lit
 }
 nodes [
  Constant3Vector() {constant=<rgba 1 1 1 1>}
  Constant3Vector() {constant=<rgba 0.2716 0.276 0.2726 1>}
  Constant3Vector() {constant=<rgba 1 1 0.998 1>}
  VectorParameter() {defaultValue=<rgba 3.8749 6.8065 30 0>}
  VectorParameter() {parameterName=Param_1, defaultValue=<rgba 0.3594 0.3594 0.3587 1>}
 ]
}
