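# Holographic Apollo island map table material for the front end, showing the minimap as a tinted, glitching digital projection.
# Stepped digital-noise blocks, seeded by object position, distort minimap UVs, dim color and feed world position offset alongside slow Z-axis rotation.
# Animated distance rings, edge glow and Fresnel add emissive light, while vertex red fades edges.
# ShowClear suppresses glitches and rotation and adjusts rings; several branches, including scanlines and water level, appear unused.
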Material /BRCosmetics/Environments/FrontEnd/Materials/BP12/M_Apollo_MapTable_Holo
graph {
 properties {bTangentSpaceNormal=false, pixelDepthOffsetMode=PDOM_Legacy, usedShadingModels=Default Lit}
 nodes [
  n1 = TextureCoordinate()
  n2 = MaterialFunctionCall() {materialFunction=/Engine/Functions/Engine_MaterialFunctions02/WorldPositionOffset/ObjectScale}
  n3 = AppendVector(A=@n2.Scale X, B=@n2.Scale X)
  n4 = Multiply(A=@n1, B=Multiply(A=ScalarParameter() {defaultValue=6, parameterName=BlockTiling}, B=@n3))
  n5 = ObjectPositionWS()
  n6 = Divide(A=Divide(A=ComponentMask(None=@n5) {r=true}, B=ComponentMask(None=@n5) {g=true}), B=ComponentMask(None=@n5) {b=true})
  n7 = Add(A=Time(), B=Multiply(A=Frac(None=@n6), B=ScalarParameter() {defaultValue=23425.4395, parameterName=Random}))
  n8 = Multiply(A=@n7, B=ScalarParameter() {defaultValue=12.0689, parameterName=Block Speed A})
  n9 = TextureSample(UVs=Panner(Coordinate=@n4, Time=Ceil(None=@n8)) {bFractionalPart=true, speedX=-0.477, speedY=-0.334}) {samplerType=SAMPLERTYPE_LinearColor, texture=/Game/Global/Textures/Default/Noise/T_Atlas_DigitalNoise_01}
  n10 = ScalarParameter() {defaultValue=0.5867, parameterName=LageBlockMask}
  n11 = Ceil(None=Multiply(A=@n7, B=ScalarParameter() {defaultValue=1, parameterName=SpeedWaittilSwitch}))
  n12 = Constant() {r=0.4563}
  n13 = TextureSample(UVs=Add(A=Multiply(A=@n4, B=@n10), B=Multiply(A=@n11, B=@n12))) {samplerType=SAMPLERTYPE_LinearColor, texture=&T_Atlas_DigitalNoise_01}
  n14 = Multiply(A=@n13.R, B=ScalarParameter() {defaultValue=0.1, parameterName=MinGlitch})
  n15 = Frac(None=Multiply(A=@n7, B=ScalarParameter() {defaultValue=0.5, parameterName=TimeSpeed}))
  n16 = ScalarParameter() {defaultValue=3, parameterName=WaitTime}
  n17 = Floor(None=Clamp(None=Multiply(A=@n15, B=@n16)))
  n18 = LinearInterpolate(
   A=ScalarParameter() {defaultValue=0.2327, parameterName=GlitchBrightness}
   B=Max(A=@n9.R, B=ScalarParameter() {defaultValue=0.2667, parameterName=GlitchBrightness2})
   Alpha=Multiply(A=LinearInterpolate(A=@n13.R, B=@n14, Alpha=@n17), B=ScalarParameter() {defaultValue=1, parameterName=BlockOpacity})
  ) {constA=1}
  n19 = ScalarParameter() {defaultValue=0.02, parameterName=Distort}
  n20 = ScalarParameter() {defaultValue=80, parameterName=C1}
  n21 = ScalarParameter() {parameterName=C2}
  n22 = Multiply(A=ScalarParameter() {defaultValue=1.8832, parameterName=BlockTiling2}, B=@n3)
  n23 = Multiply(A=@n7, B=ScalarParameter() {defaultValue=0.4571, parameterName=Block Speed Mask})
  n24 = TextureSample(UVs=Panner(Coordinate=Multiply(A=@n1, B=@n22), Time=Ceil(None=@n23)) {bFractionalPart=true, speedX=-0.477, speedY=-0.334}) {samplerType=SAMPLERTYPE_LinearColor, texture=&T_Atlas_DigitalNoise_01}
  n25 = ScalarParameter() {parameterName=ShowClear}
  n26 = Multiply(A=Saturate(None=LinearInterpolate(A=@n20, B=@n21, Alpha=@n24.R)), B=OneMinus(None=@n25))
  n27 = Add(A=LinearInterpolate(A=Constant(), B=Multiply(A=@n18, B=@n19), Alpha=@n26), B=Clamp(None=TextureCoordinate()))
  n28 = TextureSampleParameter2D(UVs=@n27) {parameterName=MinimapDiscoveryMask, texture=/BRCosmetics/Athena/Apollo/Maps/UI/Apollo_Terrain_Minimap_Undiscovered}
  n29 = Multiply(A=VectorParameter() {parameterName=DigitalMapColor, defaultValue=<rgba 0.4958 1.2272 2 1>}.RGB, B=@n28.RGB)
  n30 = ScalarParameter() {defaultValue=3.2258, parameterName=ContrastDigital}
  Add(
   A=@n29
   B=Multiply(A=VectorParameter() {parameterName=DigitalMapColor2, defaultValue=<rgba 1000 19.4855 0 1>}.RGB, B=Power(Base=@n28.RGB, Exp=@n30))
  )
  n31 = Multiply(A=TextureCoordinate(), B=ScalarParameter() {defaultValue=4.473, parameterName=Grid_Scale})
  n32 = Time()
  n33 = ScalarParameter() {defaultValue=0.1619, parameterName=RingSpeed}
  n34 = TextureSample() {samplerType=SAMPLERTYPE_Normal, texture=/BRCosmetics/Environments/FrontEnd/Textures/BP12/T_Apollo_MapTable_Holo_N}
  n35 = MaterialFunctionCall() {materialFunction=&ObjectScale}
  n36 = ScalarParameter() {defaultValue=1, parameterName=DistortLinesAltitude}
  n37 = Multiply(A=@n35.Scale X, B=AppendVector(A=AppendVector(A=@n36, B=@n36), B=Constant()))
  n38 = Distance(A=ObjectPositionWS(), B=Add(A=WorldPosition().XYZ, B=Multiply(A=@n34.RGB, B=@n37)))
  n39 = ScalarParameter() {defaultValue=98.8549, parameterName=RingFrequency}
  n40 = Add(A=Frac(None=Multiply(A=@n32, B=@n33)), B=Divide(A=@n38, B=Multiply(A=@n35.Scale X, B=@n39)))
  n41 = Saturate(None=LinearInterpolate(
    A=ScalarParameter() {defaultValue=-203.9537, parameterName=RingContrast1}
    B=ScalarParameter() {defaultValue=1, parameterName=RingContrast2}
    Alpha=Sine(None=@n40)
   ))
  n42 = ScalarParameter() {defaultValue=0.2, parameterName=DarkenRings}
  n43 = Multiply(
   A=VectorParameter() {parameterName=EmissiveGridColor, defaultValue=<rgba 0.0283 0 1 1>}.RGB
   B=LinearInterpolate(A=@n41, B=Multiply(A=@n42, B=@n41), Alpha=@n25)
  )
  Add(
   A=Multiply(
    A=VectorParameter() {parameterName=EmissiveGridColor2, defaultValue=<rgba 0 0 0 1>}.RGB
    B=TextureSample(UVs=@n31) {samplerType=SAMPLERTYPE_LinearGrayscale, texture=/Game/Packages/Fortress_SharedMaterials/Textures/T_Grid__M}.R
   )
   B=@n43
  )
  n44 = WorldPosition() {worldPositionShaderOffset=WPT_ExcludeAllShaderOffsets}
  n45 = Transform(None=@n34.RGB)
  n46 = ScalarParameter() {defaultValue=1.4638, parameterName=Bias}
  n47 = Distance(A=ObjectPositionWS(), B=Add(A=@n44.XYZ, B=Multiply(A=@n45, B=@n46)))
  Add(A=ComponentMask(None=@n44.XYZ) {b=true}, B=Multiply(A=@n47, B=@n46))
  n48 = MaterialFunctionCall!(UVsOverride=@n27, TextureOverride) {materialFunction=/Game/Athena/Apollo/Maps/UI/MF_Minimap}
  n49 = ScalarParameter() {defaultValue=2, parameterName=ColorC}
  n50 = ScalarParameter() {defaultValue=1.5, parameterName=Digital}
  n51 = LinearInterpolate(
   A=@n48.RGB
   B=@n29
   Alpha=ShadingPathSwitch(Default=@n50, Deferred=@n50, Forward=@n50, Mobile=ScalarParameter() {defaultValue=1.1083, parameterName=Digital_mobile})
  )
  Add(
   A=Multiply(A=Power(Base=@n48.RGB, Exp=@n49), B=ScalarParameter() {defaultValue=3, parameterName=ColorB})
   B=Multiply(A=@n51, B=ScalarParameter() {defaultValue=0.7619, parameterName=ColorD})
  )
  Add()
  n52 = ObjectPositionWS()
  n53 = ScalarParameter() {defaultValue=-44.6614, parameterName=OffsetBase}
  n54 = MaterialFunctionCall() {materialFunction=&ObjectScale}
  n55 = Subtract(A=ComponentMask(None=WorldPosition().XYZ) {b=true}, B=Add(A=ComponentMask(None=@n52) {b=true}, B=Multiply(A=@n53, B=@n54.Scale Z)))
  n56 = Clamp(None=@n55) {clampMode=CMODE_ClampMin}
  n57 = ScalarParameter() {defaultValue=25.4531, parameterName=MaxHeight}
  n58 = Multiply(A=Divide(A=Multiply(A=@n57, B=@n54.Scale Z)) {constB=8}, B=ScalarParameter() {parameterName=CurrentWaterLevel})
  AppendVector(A=Constant2Vector(), B=Multiply(A=Clamp(None=@n56, Max=@n58) {clampMode=CMODE_ClampMax}) {constB=-1})
  n59 = Divide(
   A=TextureSample() {samplerType=SAMPLERTYPE_Masks, texture=/BRCosmetics/Environments/FrontEnd/Textures/BP12/T_Apollo_MapTable_Holo_M}.R
   B=ScalarParameter() {defaultValue=0.0681, parameterName=WaveLength}
  )
  Ceil(None=Saturate(None=Sine(None=@n59)))
  Constant3Vector()
  Constant() {r=1}
  n60 = Distance(A=ObjectPositionWS(), B=WorldPosition().XYZ)
  n61 = Time()
  n62 = ScalarParameter() {defaultValue=0.1048, parameterName=RingFadeB2Speed}
  n63 = ScalarParameter() {defaultValue=1, parameterName=RingFadeB2Pause}
  n64 = Frac(None=Multiply(A=@n61, B=Divide(A=@n62, B=@n63)))
  n65 = Add(A=Divide(A=@n60, B=ScalarParameter() {defaultValue=161.6321, parameterName=RingFadeB2}), B=Saturate(None=Multiply(A=@n64, B=@n63)))
  Cosine(None=@n65)
  n66 = Constant3Vector() {constant=<rgba 0 0 1 0>}
  n67 = VertexColor()
  LinearInterpolate(
   A=MaterialFunctionCall(BaseNormal=@n66, AdditionalNormal=@n34.RGB) {materialFunction=&BlendAngleCorrectedNormals}.Result
   B=@n66
   Alpha=@n67.R
  )
  n68 = ComponentMask(None=@n48.RGB) {b=true}
  LinearInterpolate(
   A=ScalarParameter() {defaultValue=0.0229, parameterName=Specular1}
   B=ScalarParameter() {defaultValue=-0.0076, parameterName=Specular2}
   Alpha=@n68
  )
  n69 = Divide(A=@n55, B=@n58)
  LinearInterpolate(B=@n66, Alpha=Saturate(None=OneMinus(None=@n69)))
  LinearInterpolate(A=ScalarParameter() {defaultValue=0.6952, parameterName=Roughness1}, B=ScalarParameter() {parameterName=Roughness2}, Alpha=@n68)
  LinearInterpolate(
   A=ScalarParameter() {defaultValue=0.05, parameterName=Blocks Dark}
   B=ScalarParameter() {defaultValue=1, parameterName=Blocks Light}
   Alpha=@n9.R
  )
  n70 = VectorParameter() {parameterName=wiggle1, defaultValue=<rgba 0.2 0.4 0.3524 1>}
  n71 = VectorParameter() {parameterName=wiggle1_1, defaultValue=<rgba 0 1 0 1>}
  MaterialFunctionCall!(
   Number of Scanlines=ScalarParameter() {defaultValue=150, parameterName=Scanline Amount}
   Scanline Speed
   "Scanlines Thickness (Bias)"=ScalarParameter() {defaultValue=2, parameterName=ScanlineSize}
   Scanlines Dark Value=ScalarParameter() {defaultValue=0.5, parameterName=Scanlines Dark}
   Scanlines Light Value=ScalarParameter() {defaultValue=1, parameterName=Scanlines Light}
   Block Opacity=StaticSwitchParameter(
    True=MaterialFunctionCall!(WPO Intensity A=@n70.RGB, WPO Intensity B=@n71.RGB, Time, Bias) {materialFunction=/Game/Effects/Fort_Effects/Effects/Atlas/Materials/MF_Atlas_HologramWiggle_01}.Intense 0-1
    False=Constant()
   ) {defaultValue=true, parameterName=Use Blocks?}
   UVs
   Time
   Block Repeat
  ) {materialFunction=/Game/Effects/Fort_Effects/Effects/Atlas/Materials/MF_Atlas_Scanlines_01}
  Multiply(A=@n7)
  Multiply(A=@n13.R, B=@n17)
  n72 = Divide(A=@n47, B=ScalarParameter() {defaultValue=17.4241, parameterName=Width})
  n73 = Saturate(None=Subtract(A=Sine(None=@n72), B=ScalarParameter() {defaultValue=0.9179, parameterName=Shift}))
  Multiply(B=LinearInterpolate(
    A=VectorParameter() {parameterName=Colormult, defaultValue=<rgba 0 0 0 1>}.RGB
    B=Constant3Vector() {constant=<rgba 1 1 1 0>}
    Alpha=Ceil(None=@n73)
   ))
  Multiply(A=@n61, B=ScalarParameter() {defaultValue=1.0095, parameterName=RingNoiseSpeed})
  OneMinus(None=@n17)
  n74 = ScalarParameter() {defaultValue=0.2667, parameterName=GirdPan_Speed}
  Panner(Coordinate=@n31, Speed=AppendVector(A=@n74, B=@n74))
  n75 = Multiply(A=LinearInterpolate(A=Constant(), B=@n18, Alpha=@n26), B=VectorParameter() {parameterName=WPOOFFSet, defaultValue=<rgba 0 0.9229 10 1>}.RGB)
  n76 = MaterialFunctionCall() {materialFunction=&ObjectScale}
  n77 = Multiply(A=ScalarParameter() {defaultValue=214.1731, parameterName=EdgeRingOffset}, B=@n76.Scale X)
  n78 = LinearInterpolate(
   A=ScalarParameter() {defaultValue=394.4009, parameterName=EdgeRingBlend}
   B=ScalarParameter() {defaultValue=403.4466, parameterName=EdgeRingBlend2}
   Alpha=@n25
  )
  n79 = Saturate(None=Divide(A=Subtract(A=@n38, B=@n77), B=Multiply(A=@n76.Scale X, B=@n78)))
  n80 = OneMinus(None=@n79)
  n81 = Constant3Vector() {constant=<rgba 0 0 1 0>}
  n82 = Frac(None=Multiply(A=Time(), B=ScalarParameter() {defaultValue=0.025, parameterName=RotationSpeed}))
  n83 = LinearInterpolate(A=@n82, B=Round(None=@n82), Alpha=@n25) {constB=0}
  n84 = MaterialFunctionCall() {materialFunction=&ObjectPivotPoint}
  n85 = WorldPosition()
  Reroute(None=Add(
    A=Multiply(A=@n75, B=@n80)
    B=RotateAboutAxis(NormalizedRotationAxis=@n81, RotationAngle=@n83, PivotPoint=@n84.Object Pivot Location, Position=@n85.XYZ)
   ))
  n86 = OneMinus(None=@n18)
  n87 = ScalarParameter() {defaultValue=0.4286, parameterName=GlitchBrightness3}
  n88 = ScalarParameter() {defaultValue=1, parameterName=MapBrightness}
  n89 = ShadingPathSwitch(Default=@n88, Deferred=@n88, Forward=@n88, Mobile=ScalarParameter() {defaultValue=1.5, parameterName=MapBrightness_Mobile})
  n90 = Saturate(None=Divide(A=@n60, B=ScalarParameter() {defaultValue=265.8745, parameterName=RingFadeB}))
  n91 = LinearInterpolate(
   A=LinearInterpolate(A=@n51, B=LinearInterpolate(A=@n48.RGB, B=@n51, Alpha=@n90), Alpha=@n25)
   B=VectorParameter() {parameterName=FadeEdgeColor, defaultValue=<rgba 0 0.093 0.4531 1>}.RGB
   Alpha=@n67.R
  )
  n92 = Multiply(
   A=LinearInterpolate(A=Constant() {r=1}, B=Max(A=@n86, B=@n87), Alpha=@n26)
   B=Power(Base=Multiply(A=@n89, B=@n91), Exp=ScalarParameter() {defaultValue=1.25, parameterName=MapContrast})
  )
  n93 = VectorParameter() {parameterName=EmissiveEdgeColor, defaultValue=<rgba 0 0.5523 2 1>}
  n94 = ScalarParameter() {defaultValue=2.1462, parameterName=EdgeRingContrast2}
  n95 = ScalarParameter() {defaultValue=10, parameterName=EdgeRingBlend2_Closeup}
  n96 = Add(
   A=Power(Base=@n79, Exp=LinearInterpolate(A=@n94, B=@n95, Alpha=@n25))
   B=Power(Base=@n79, Exp=ScalarParameter() {defaultValue=160.2474, parameterName=EdgeRingContrast})
  )
  n97 = ScalarParameter() {defaultValue=9.6898, parameterName=Fresnel}
  n98 = PixelNormalWS()
  n99 = Add(
   A=Add(A=@n43, B=Multiply(A=@n93.RGB, B=@n96))
   B=Multiply(
    A=Fresnel(ExponentIn=@n97, Normal=@n98) {baseReflectFraction=0}
    B=VectorParameter() {parameterName=EmissiveFresnel, defaultValue=<rgba 0 0.0432 0.3646 1>}.RGB
   )
  )
  Reroute(None=Clamp(None=Add(A=@n92, B=@n99)))
  n100 = Add(A=Sine(None=@n65))
  n101 = OneMinus(None=Saturate(None=Divide(A=@n100)))
  n102 = ScalarParameter() {defaultValue=0.6, parameterName=RingFadeExp}
  n103 = TextureCoordinate()
  n104 = ScalarParameter() {defaultValue=2.3221, parameterName=RingNoiseTiling}
  n105 = MaterialFunctionCall(
   Transition Phase=Multiply(A=ScalarParameter() {defaultValue=1, parameterName=RingFadeBrightness}, B=Power(Base=@n101, Exp=@n102))
   Height Texture=TextureSample(UVs=Multiply(A=@n103, B=@n104)) {samplerType=SAMPLERTYPE_LinearColor, texture=&T_Atlas_DigitalNoise_01}.R
  ) {materialFunction=&HeightLerp}
  Saturate(None=Add(A=Subtract(A=@n90, B=@n105.Alpha), B=ScalarParameter() {defaultValue=-0.1857, parameterName=RingFadeMaxBlue}))
  TextureCoordinate()
  Transform(None=@n66) {transformSourceType=TRANSFORMSOURCE_World, transformType=TRANSFORM_Tangent}
 ]
}
