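# Front-end prop material for a movie memorabilia mesh, with a procedural bobblehead-style bob and wiggle vertex animation.
# A time-driven sine moves vertices along Z, while two chained RotateAboutAxis nodes swing them around a pivot parameter on separate axes and speeds.
# The summed offset is scaled by an Animation parameter, zero by default, and masked by vertex color red.
# Diffuse, normal and mask textures come from the OilRig prop set.
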Material /BRCosmetics/Environments/FrontEnd/Materials/BP12/M_BobHeadInteraction
graph {
 properties {pixelDepthOffsetMode=PDOM_Legacy, usedShadingModels=Default Lit}
 nodes [
  n1 = Multiply(A=Time(), B=ScalarParameter() {defaultValue=4.5245, parameterName=UpdownSpeed})
  n2 = Multiply(
   A=Multiply(A=Constant3Vector() {constant=<rgba 0 0 1 0>}, B=Sine(None=@n1))
   B=ScalarParameter() {defaultValue=2, parameterName=UpdownDistance}
  )
  n3 = Time()
  n4 = Sine(None=Multiply(A=@n3, B=ScalarParameter() {defaultValue=3, parameterName=WiggleSpped}))
  n5 = LinearInterpolate(A=ScalarParameter() {parameterName=Rotation_1}, B=ScalarParameter() {defaultValue=0.02, parameterName=Rotation_2})
  n6 = TransformPosition(None=VectorParameter() {parameterName=PivotPosition, defaultValue=<rgba 0 0 40.878 1>}.RGB) {transformType=TRANSFORMPOSSOURCE_World}
  n7 = Add(B=WorldPosition().XYZ)
  n8 = RotateAboutAxis(
   NormalizedRotationAxis=Normalize(VectorInput=VectorParameter() {parameterName=RotationAxis, defaultValue=<rgba 1 0 -0.1524 1>}.RGB)
   RotationAngle=Multiply(A=Multiply(A=@n4, B=@n5))
   PivotPoint=@n6
   Position=@n7
  )
  n9 = VectorParameter() {parameterName=RotationAxis2, defaultValue=<rgba 0 1 0.2095 1>}
  n10 = Sine(None=Multiply(A=@n3, B=ScalarParameter() {defaultValue=4, parameterName=WiggleSpped2}))
  n11 = Add(
   A=@n8
   B=RotateAboutAxis(NormalizedRotationAxis=Normalize(VectorInput=@n9.RGB), RotationAngle=Multiply(A=@n5, B=@n10), PivotPoint=@n6, Position=Add(A=@n8, B=@n7))
  )
  Multiply(A=VertexColor().R, B=Multiply(A=ScalarParameter() {parameterName=Animation}, B=Add(A=@n2, B=@n11)))
  TextureSample() {texture=/Game/Environments/Apollo/Sets/OilRig/Props/Textures/T_OilRig_MovieMemorabilia_02_D}
  TextureSample() {samplerType=SAMPLERTYPE_Normal, texture=/Game/Environments/Apollo/Sets/OilRig/Props/Textures/T_OilRig_MovieMemorabilia_02_N}
  TextureSample() {samplerType=SAMPLERTYPE_Masks, texture=/Game/Environments/Apollo/Sets/OilRig/Props/Textures/T_OilRig_MovieMemorabilia_02_S}
 ]
}
