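# Two-sided lit spline-mesh material for cables in the BP12 front-end environment, with optional wind sway.
# SplineThicken widens the mesh by a Width parameter, and its normal is blended with a V-based gradient normal.
# UseWindSwing adds sine motion along WindDirection, phase-offset by object position and masked along U to peak mid-span with pinned ends.
# The two summed sines share identical inputs, effectively doubling amplitude.
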Material /BRCosmetics/Environments/FrontEnd/Materials/BP12/M_Cables_Spline
graph {
 properties {bUsedWithSplineMeshes=true, pixelDepthOffsetMode=PDOM_Legacy, twoSided=true, usedShadingModels=Default Lit}
 nodes [
  n1 = ScalarParameter() {defaultValue=4.9597, parameterName=Width}
  n2 = MaterialFunctionCall(WidthBase=@n1, WidthTip=@n1, Expand U or V UV Channel=StaticBool()) {materialFunction=&SplineThicken}
  n3 = Constant3Vector() {constant=<rgba 0 0 1 0>}
  n4 = Constant3Vector() {constant=<rgba 0 1 0 0>}
  n5 = ComponentMask(None=TextureCoordinate()) {g=true}
  MaterialFunctionCall(
   BaseNormal=@n2.Normal
   AdditionalNormal=LinearInterpolate(A=LinearInterpolate(A=@n3, B=@n4, Alpha=@n5), B=Constant3Vector() {constant=<rgba 0 -1 0 0>}, Alpha=OneMinus(None=@n5))
  ) {materialFunction=&BlendAngleCorrectedNormals}
  ScalarParameter() {defaultValue=0.3238, parameterName=R}
  n6 = ComponentMask(None=TextureCoordinate()) {r=true}
  n7 = Multiply(A=@n6, B=OneMinus(None=@n6))
  n8 = MaterialFunctionCall(Float3=Divide(A=ObjectPositionWS()) {constB=1.33}) {materialFunction=&BreakOutFloat3Components}
  n9 = Add(A=Add(A=@n8.R, B=@n8.R), B=@n8.B)
  n10 = Add(A=Time(), B=Multiply(A=Frac(None=@n9), B=ScalarParameter() {defaultValue=24.324, parameterName=Random}))
  n11 = Sine(None=Multiply(A=@n10, B=ScalarParameter() {defaultValue=1, parameterName=WindSpeed}))
  n12 = Sine(None=Multiply(A=@n10, B=ScalarParameter() {defaultValue=1, parameterName=WindSpeed}))
  n13 = Multiply(
   A=Power(Base=Multiply(A=@n7) {constB=4}, Exp=ScalarParameter() {defaultValue=2, parameterName=WindFalloff})
   B=Multiply(A=Add(A=@n11, B=@n12), B=ScalarParameter() {defaultValue=1, parameterName=WindStrength})
  )
  n14 = VectorParameter() {parameterName=WindDirection, defaultValue=<rgba 1 0 0 1>}
  StaticSwitchParameter(True=Add(A=@n2.WorldPositionOffset, B=Multiply(A=@n13, B=@n14.RGB)), False=@n2.WorldPositionOffset) {parameterName=UseWindSwing}
  VectorParameter() {parameterName=Color, defaultValue=<rgba 0.0156 0.0156 0.0156 1>}
 ]
}
