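# Masked, two-sided Default Lit material for spline meshes, apparently colored twine in the BP12 front-end environment.
# Procedural normals come from powered UV gradients for cross-section rounding and tiled, rotated Frac bands for strand detail.
# These are angle-corrected blended onto the SplineThicken normal, driven by a shared Width.
# The same gradients combine into mask terms with tunable parameters, alongside a Color parameter; final material input connections are not shown.
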Material /BRCosmetics/Environments/FrontEnd/Materials/BP12/M_Colored_Twine_spline
graph {
 properties {bUsedWithSplineMeshes=true, blendMode=BLEND_Masked, pixelDepthOffsetMode=PDOM_Legacy, twoSided=true, usedShadingModels=Default Lit}
 nodes [
  n1 = ScalarParameter() {defaultValue=1.8867, parameterName=Width}
  n2 = ComponentMask(None=TextureCoordinate()) {g=true}
  n3 = ScalarParameter() {defaultValue=3, parameterName=RoungNess}
  n4 = Power(Base=@n2, Exp=@n3)
  n5 = LinearInterpolate(A=Constant3Vector() {constant=<rgba 0 0 1 0>}, B=Constant3Vector() {constant=<rgba 0 -1 0 0>}, Alpha=@n4)
  n6 = Constant3Vector() {constant=<rgba 0 1 0 0>}
  n7 = Power(Base=OneMinus(None=@n2), Exp=@n3)
  n8 = Multiply(A=TextureCoordinate(), B=ScalarParameter() {defaultValue=8, parameterName=Tiling})
  n9 = ScalarParameter() {defaultValue=1.5705, parameterName=Rotation}
  n10 = Frac(None=ComponentMask(None=Rotator(Coordinate=@n8, Time=@n9)) {g=true})
  n11 = Power(Base=@n10, Exp=@n3)
  n12 = Power(Base=OneMinus(None=@n10), Exp=@n3)
  n13 = LinearInterpolate(
   A=LinearInterpolate(A=Constant3Vector() {constant=<rgba 0 0 1 0>}, B=Constant3Vector() {constant=<rgba 1 1 0 0>}, Alpha=@n11)
   B=Constant3Vector() {constant=<rgba -1 -1 0 0>}
   Alpha=@n12
  )
  MaterialFunctionCall(
   BaseNormal=MaterialFunctionCall(WidthBase=@n1, WidthTip=@n1, Expand U or V UV Channel=StaticBool()) {materialFunction=&SplineThicken}.Normal
   AdditionalNormal=MaterialFunctionCall(BaseNormal=LinearInterpolate(A=@n5, B=@n6, Alpha=@n7), AdditionalNormal=Normalize(VectorInput=@n13)) {materialFunction=&BlendAngleCorrectedNormals}.Result
  ) {materialFunction=&BlendAngleCorrectedNormals}
  n14 = Multiply(A=@n4, B=@n7)
  n15 = Multiply(A=@n11, B=@n12)
  Multiply(A=@n14, B=@n15)
  n16 = ScalarParameter() {defaultValue=1, parameterName=Masked2}
  n17 = ScalarParameter() {parameterName=Masked3}
  Multiply(A=Add(A=Multiply(A=@n14, B=@n16), B=Add(A=@n15, B=@n17)), B=ScalarParameter() {defaultValue=1, parameterName=Masked})
  ScalarParameter() {defaultValue=-0.0095, parameterName=R}
  VectorParameter() {parameterName=Color, defaultValue=<rgba 1 0 0.0076 1>}
 ]
}
