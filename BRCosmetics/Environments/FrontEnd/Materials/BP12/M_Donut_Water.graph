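# Unlit, two-sided translucent water material for the BP12 front-end environment.
# Two panning world-XY normal maps are blended and flattened to bend a reflection vector into a day-sky cubemap, whose contrast-adjusted red channel is tinted and added to a base color.
# Nested depth fades compute opacity that varies near intersecting geometry.
# Summed directional sine waves form a vertical offset, alongside apparent leftover nodes such as world-aligned normals.
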Material /BRCosmetics/Environments/FrontEnd/Materials/BP12/M_Donut_Water
graph {
 properties {
  blendMode=BLEND_Translucent
  pixelDepthOffsetMode=PDOM_Legacy
  shadingModel=MSM_Unlit
  translucencyLightingMode=TLM_SurfacePerPixelLighting
  twoSided=true
  usedShadingModels=Unlit
 }
 nodes [
  n1 = VectorParameter() {parameterName=Color, defaultValue=<rgba 0.0332 0.054 0.0729 1>}
  n2 = ComponentMask(None=WorldPosition() {worldPositionShaderOffset=WPT_ExcludeAllShaderOffsets}.XYZ) {g=true, r=true}
  n3 = ScalarParameter() {defaultValue=321.7803, parameterName=Size}
  n4 = Multiply(A=Time(), B=ScalarParameter() {defaultValue=3, parameterName=Speed})
  n5 = Panner(
   Coordinate=Divide(A=@n2, B=@n3)
   Time=@n4
   Speed=AppendVector(A=ScalarParameter() {defaultValue=0.01, parameterName=Pan2x}, B=ScalarParameter() {defaultValue=0.02, parameterName=Pan2y})
  ) {bFractionalPart=true}
  n6 = Multiply(A=@n3, B=ScalarParameter() {defaultValue=1.381, parameterName=Size2})
  n7 = Panner(
   Coordinate=Divide(A=@n2, B=@n6)
   Time=@n4
   Speed=AppendVector(A=ScalarParameter() {defaultValue=-0.015, parameterName=Pan1x}, B=ScalarParameter() {defaultValue=-0.018, parameterName=Pan1y})
  ) {bFractionalPart=true}
  n8 = MaterialFunctionCall(
   Normal=MaterialFunctionCall(
    BaseNormal=TextureSample(UVs=@n5) {samplerType=SAMPLERTYPE_Normal, texture=/FortWater/Textures/T_Water_TilingNormal_Waves_01}.RGB
    AdditionalNormal=TextureSample(UVs=@n7) {samplerType=SAMPLERTYPE_Normal, texture=&T_Water_TilingNormal_Waves_01}.RGB
   ) {materialFunction=&BlendAngleCorrectedNormals}.Result
   Flatness=ScalarParameter() {defaultValue=0.8286, parameterName=Normal}
  ) {materialFunction=&FlattenNormal}
  n9 = TextureSample(UVs=ReflectionVectorWS(CustomWorldNormal=@n8.Result)) {texture=/Game/Packages/Fortress_Sky/SkyDome/MaterialInstances/SkyDome_Apollo/Cubemap_Apollo_Day}
  n10 = Multiply(
   A=Power(Base=@n9.R, Exp=ScalarParameter() {defaultValue=2.3162, parameterName=Contrast})
   B=VectorParameter() {parameterName=ReflectionColor, defaultValue=<rgba 2.0312 2.3901 3 1>}.RGB
  )
  Add(A=@n1.RGB, B=@n10)
  n11 = DotProduct(
   A=Normalize(VectorInput=VectorParameter() {parameterName=WaveDirection_A, defaultValue=<rgba 1 1 0 1>}.RGB)
   B=WorldPosition() {worldPositionShaderOffset=WPT_ExcludeAllShaderOffsets}.XYZ
  )
  n12 = Multiply(A=Time(), B=ScalarParameter() {defaultValue=42.3988, parameterName=WaveSpeed1})
  n13 = Add(A=ComponentMask(None=@n11) {r=true}, B=@n12)
  n14 = Sine(None=Divide(A=ComponentMask(None=@n13) {r=true}, B=ScalarParameter() {defaultValue=145.9291, parameterName=WaveLength1}))
  n15 = ComponentMask(None=@n11) {g=true}
  n16 = Sine(None=Divide(A=Add(A=@n15, B=@n12), B=ScalarParameter() {defaultValue=156.9975, parameterName=WaveLength2}))
  AppendVector(A=Constant2Vector(), B=Multiply(A=Add(A=@n14, B=@n16), B=ScalarParameter() {defaultValue=4, parameterName=Amplitude}))
  ComponentMask(None=WorldPosition().XYZ) {g=true, r=true}
  MaterialFunctionCall!(
   TextureObject=TextureObject() {samplerType=SAMPLERTYPE_Normal, texture=/Game/Packages/Fortnite_UniqueMaterials_1/Textures/T_Water_TilingNormal_Waves_02}
   TextureSize=@n6
   Normal
   WorldSpace
   ProjectionTransitionContrast
   WorldPosition
   Use High Quality Normals
   Pan_YZ
   Pan_XZ
   Pan_XY
  ) {materialFunction=/Game/Weapons/WeaponSkins/Wraps/MaterialFunctions/WorldAlignedNormal_ComplexPan}
  MaterialFunctionCall(
   TextureObject=TextureObject() {samplerType=SAMPLERTYPE_Normal, texture=&T_Water_TilingNormal_Waves_02}
   TextureSize=@n3
  ) {materialFunction=&WorldAlignedNormal_ComplexPan}
  Multiply(A=@n1.RGB, B=@n9.RGB)
  n17 = ScalarParameter() {defaultValue=0.771, parameterName=Opacity}
  n18 = ScalarParameter() {defaultValue=0.99, parameterName=Opacity2}
  n19 = DepthFade(Opacity=Constant() {r=1}, FadeDistance=ScalarParameter() {defaultValue=208.0963, parameterName=DepthFade2})
  n20 = Saturate(None=DepthFade(Opacity=LinearInterpolate(A=@n17, B=@n18, Alpha=@n19), FadeDistance=ScalarParameter() {defaultValue=5, parameterName=DepthFade}))
  Multiply(A=Divide(A=Constant() {r=1}, B=OneMinus(None=@n20)), B=@n10)
  ScalarParameter() {defaultValue=0.4, parameterName=Spec}
  ScalarParameter() {defaultValue=0.1143, parameterName=Roughness}
  Sine()
  Transform(None=@n8.Result) {transformSourceType=TRANSFORMSOURCE_World, transformType=TRANSFORM_Tangent}
 ]
}
