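# Lit prop material that bobs a whole object vertically as if floating, with clothes washer textures as defaults.
# Object position dotted with a normalized wave direction, plus scaled time, feeds two sine waves with separate wavelengths and a shared speed.
# Their amplitude-scaled sum drives Z-only World Position Offset.
# The Spec mask's green channel feeds Metallic, Specular and Roughness alike.
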Material /BRCosmetics/Environments/FrontEnd/Materials/BP12/M_Floatingobject
graph {
 properties {bUseMaterialAttributes=true, pixelDepthOffsetMode=PDOM_Legacy, usedShadingModels=Default Lit}
 nodes [
  n1 = TextureSampleParameter2D() {parameterName=Spec, samplerType=SAMPLERTYPE_Masks, texture=/Game/Environments/Props/Residential/Textures/Prop_ClothesWasher_02_S}
  n2 = TextureSampleParameter2D() {parameterName=Normal, samplerType=SAMPLERTYPE_Normal, texture=/Game/Environments/Props/Residential/Textures/Prop_ClothesWasher_02_N}
  n3 = DotProduct(A=Normalize(VectorInput=VectorParameter() {parameterName=WaveDirection_A, defaultValue=<rgba 1 1 0 1>}.RGB), B=ObjectPositionWS())
  n4 = Multiply(A=Time(), B=ScalarParameter() {defaultValue=42.3988, parameterName=WaveSpeed1})
  n5 = ComponentMask(None=Add(A=ComponentMask(None=@n3) {r=true}, B=@n4)) {r=true}
  n6 = ScalarParameter() {defaultValue=145.9291, parameterName=WaveLength1}
  n7 = Add(A=ComponentMask(None=@n3) {g=true}, B=@n4)
  n8 = ScalarParameter() {defaultValue=156.9975, parameterName=WaveLength2}
  n9 = Add(A=Sine(None=Divide(A=@n5, B=@n6)), B=Sine(None=Divide(A=@n7, B=@n8)))
  n10 = ScalarParameter() {defaultValue=4, parameterName=Amplitude}
  MakeMaterialAttributes!(
   BaseColor=TextureSampleParameter2D() {parameterName=Diffuse, texture=/Game/Environments/Props/Residential/Textures/Prop_ClothesWasher_02_D}.RGB
   Metallic=@n1.G
   Specular=@n1.G
   Roughness=@n1.G
   Anisotropy
   EmissiveColor
   Opacity
   OpacityMask
   Normal=@n2.RGB
   Tangent
   WorldPositionOffset=AppendVector(A=Constant2Vector(), B=Multiply(A=@n9, B=@n10))
   SubsurfaceColor
   ClearCoat
   ClearCoatRoughness
   AmbientOcclusion
   Refraction
   CustomizedUVs_0
   CustomizedUVs_1
   CustomizedUVs_2
   CustomizedUVs_3
   CustomizedUVs_4
   CustomizedUVs_5
   CustomizedUVs_6
   CustomizedUVs_7
   PixelDepthOffset
   ShadingModel
   Displacement
  )
  WorldPosition() {worldPositionShaderOffset=WPT_ExcludeAllShaderOffsets}
 ]
}
