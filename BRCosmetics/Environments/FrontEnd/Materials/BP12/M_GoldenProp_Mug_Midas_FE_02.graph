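# Gold material for a Midas-themed coffee mug prop in the BP12 front-end environment, using the Default Lit shading model.
# The graph has a tintable gold Color parameter, a shine roughness scalar, a constant of one and the Agency coffee mug normal map.
# It has no math nodes or color texture, and the output wiring is not shown.
# Usage flags cover skeletal meshes, mesh particles and mesh deformers.
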Material /BRCosmetics/Environments/FrontEnd/Materials/BP12/M_GoldenProp_Mug_Midas_FE_02
graph {
 properties {
  bUsedWithMeshDeformer=true
  bUsedWithMeshParticles=true
  bUsedWithSkeletalMesh=true
  pixelDepthOffsetMode=PDOM_Legacy
  usedShadingModels=Default Lit
 }
 nodes [
  Constant() {r=1}
  ScalarParameter() {defaultValue=0.4, parameterName=Shine Roughnmess}
  TextureSample() {samplerType=SAMPLERTYPE_Normal, texture=/Game/Environments/Apollo/Sets/Agency/Props/Textures/T_Agency_CoffeeMug_01_N}
  VectorParameter() {parameterName=Color, defaultValue=<rgba 1 0.5683 0.2174 1>}
 ]
}
