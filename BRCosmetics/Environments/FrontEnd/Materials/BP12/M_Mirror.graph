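# M_Mirror is a Default Lit front-end environment material for a mirror that blends from intact to cracked through a Broken scalar.
# Broken lerps from a flat normal to a broken-mirror normal passed through FlattenNormal, and from zero to a crack-texture glow scaled by Emissive.
# The cracked-ice red channel also masks alternate R and M values, apparently roughness and metallic, when broken.
# Scale and offset parameters position the shared crack UVs.
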Material /BRCosmetics/Environments/FrontEnd/Materials/BP12/M_Mirror
graph {
 properties {pixelDepthOffsetMode=PDOM_Legacy, usedShadingModels=Default Lit}
 nodes [
  n1 = VectorParameter() {parameterName=scale, defaultValue=<rgba 1 0.6952 0 1>}
  n2 = Add(
   A=Multiply(A=TextureCoordinate(), B=ComponentMask(None=@n1.RGB) {g=true, r=true})
   B=ComponentMask(None=VectorParameter() {parameterName=offset, defaultValue=<rgba -0.1333 0 0 1>}.RGB) {g=true, r=true}
  )
  n3 = ScalarParameter() {parameterName=Broken}
  LinearInterpolate(
   A=Constant3Vector() {constant=<rgba 0 0 1 0>}
   B=MaterialFunctionCall(
    Normal=TextureSample(UVs=@n2) {samplerType=SAMPLERTYPE_Normal, texture=/BRCosmetics/Environments/FrontEnd/Textures/BP12/T_MirrorBroken_N}.RGB
    Flatness=ScalarParameter() {defaultValue=-1.0896, parameterName=Normal}
   ) {materialFunction=&FlattenNormal}.Result
   Alpha=@n3
  )
  n4 = TextureSample(UVs=@n2) {texture=/Game/Effects/Fort_Effects/Textures/Tile/T_Cracked_Ice_D}
  LinearInterpolate(A=Constant(), B=Multiply(A=@n4.R, B=ScalarParameter() {defaultValue=0.6229, parameterName=Emissive}), Alpha=@n3)
  n5 = ScalarParameter() {parameterName=R}
  LinearInterpolate(A=@n5, B=LinearInterpolate(A=@n5, B=ScalarParameter() {defaultValue=11.443, parameterName=R2}, Alpha=@n4.R), Alpha=@n3)
  n6 = ScalarParameter() {defaultValue=1, parameterName=M}
  LinearInterpolate(A=@n6, B=LinearInterpolate(A=@n6, B=ScalarParameter() {parameterName=M2}, Alpha=@n4.R), Alpha=@n3)
  TextureSample() {samplerType=SAMPLERTYPE_Normal, texture=/Game/Effects/Fort_Effects/Textures/Tile/T_Glass_Cracks_N}
  VectorParameter() {parameterName=Color, defaultValue=<rgba 1 1 1 1>}
 ]
}
