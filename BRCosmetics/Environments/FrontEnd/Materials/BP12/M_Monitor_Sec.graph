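# Front-end (BP12) security monitor material that cycles through frames of a monitor atlas with glitchy distortion and static.
# Summed object position offsets time per instance, driving both the sub-image frame choice and a scrolling Frac/Sine band that distorts the UVs.
# The sampled frame is desaturated, vignetted by a SphereMask, contrast-adjusted and multiplied by an overbright tint.
# An animated flipbook of a borrowed snowball noise texture is sharpened with Power and added on top.
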Material /BRCosmetics/Environments/FrontEnd/Materials/BP12/M_Monitor_Sec
graph {
 properties {pixelDepthOffsetMode=PDOM_Legacy, usedShadingModels=Default Lit}
 nodes [
  n1 = TextureObject() {texture=/BRCosmetics/Environments/FrontEnd/Textures/BP12/T_Monitors_Sec}
  n2 = ObjectPositionWS()
  n3 = Add(A=Add(A=ComponentMask(None=@n2) {b=true}, B=ComponentMask(None=@n2) {g=true}), B=ComponentMask(None=@n2) {r=true})
  n4 = Time()
  n5 = ScalarParameter() {defaultValue=0.266, parameterName=SpeedDistort}
  n6 = Add(A=TextureCoordinate(), B=Add(A=@n3, B=Multiply(A=@n4, B=@n5)))
  n7 = Frac(None=Divide(A=ComponentMask(None=@n6) {g=true}, B=ScalarParameter() {defaultValue=3.9345, parameterName=Frequency}))
  n8 = ScalarParameter() {defaultValue=36.8344, parameterName=break}
  n9 = Saturate(None=Multiply(A=@n7, B=@n8))
  n10 = Add(A=TextureCoordinate(), B=Multiply(A=Sine(None=@n9), B=ScalarParameter() {defaultValue=0.0951, parameterName=Distort}))
  n11 = ComponentMask(None=VectorParameter() {parameterName=Subimmages, defaultValue=<rgba 5 10 0 1>}.RGB) {g=true, r=true}
  n12 = Time()
  n13 = Floor(None=Multiply(A=Add(A=@n3, B=@n12), B=ScalarParameter() {defaultValue=0.0571, parameterName=Speed}))
  n14 = Multiply(
   A=SphereMask(
    A=TextureCoordinate()
    B=Constant2Vector() {g=0.5, r=0.5}
    Radius=ScalarParameter() {defaultValue=0.6642, parameterName=Vignette}
    Hardness=ScalarParameter() {parameterName=VignetteFalloff}
   )
   B=Desaturation(
    None=MaterialFunctionCall!(Texture=@n1, UVs=@n10, SubImages=@n11, Frame=@n13) {materialFunction=/Game/Weapons/WeaponSkins/MaterialFunction/CustomNumbers_SubUV}.RGB
    Fraction=ScalarParameter() {defaultValue=0.3333, parameterName=Desat}
   )
  )
  n15 = ScalarParameter() {defaultValue=1.3888, parameterName=Contrast}
  n16 = Multiply(A=Time(), B=ScalarParameter() {defaultValue=3, parameterName=NoiseSpeed})
  n17 = ScalarParameter() {defaultValue=2, parameterName=NoiseTiling}
  n18 = TextureCoordinate()
  n19 = ScalarParameter() {defaultValue=2.431, parameterName=NoiseScale}
  n20 = Power(
   Base=MaterialFunctionCall(
    "Animation  Phase (0-1)"=Frac(None=@n16)
    Number of Rows=@n17
    Number of Columns=@n17
    Texture=TextureObject() {texture=/Game/Abilities/Toys/Snowball/Textures/T_Snowball_Noise_01}
    UVs=Multiply(A=@n18, B=@n19)
   ) {materialFunction=&FlipBook}.Result
   Exp=ScalarParameter() {defaultValue=6.1364, parameterName=NoiseC}
  )
  Add(A=Multiply(A=Power(Base=@n14, Exp=@n15), B=VectorParameter() {parameterName=Color, defaultValue=<rgba 5 4.798 3.9583 1>}.RGB), B=@n20)
  Add(A=@n7, B=@n8)
  AppendVector(A=Constant())
  Constant3Vector()
  OneMinus(None=@n20)
  ScalarParameter() {parameterName=R}
  ScalarParameter() {defaultValue=-0.0428, parameterName=S}
 ]
}
