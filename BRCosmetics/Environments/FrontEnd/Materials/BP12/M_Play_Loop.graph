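# Opaque front-end (BP12) prop material that plays a looping "Play" flipbook as emissive, defaulting to Agency computer textures.
# A vertex-color channel mask blackens base color and gates the flipbook, whose UVs can be offset and scaled around a point.
# Switchable world position offset pushes masked vertices along normals and collapses unmasked ones to the object position.
# An optional treadmill switch blends in panning surface textures by vertex alpha.
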Material /BRCosmetics/Environments/FrontEnd/Materials/BP12/M_Play_Loop
graph {
 properties {bUseMaterialAttributes=true, bUsedWithNanite=true, pixelDepthOffsetMode=PDOM_Legacy, usedShadingModels=Default Lit}
 nodes [
  Constant()
  n1 = VertexColor()
  n2 = ChannelMaskParameter(None=AppendVector(A=@n1, B=@n1)) {parameterName=ChannelMask}
  n3 = TextureSampleParameter2D() {parameterName=Spec, samplerType=SAMPLERTYPE_Masks, texture=/Game/Environments/Apollo/Sets/Agency/Props/Textures/T_Agency_Computer_01_S}
  n4 = TextureCoordinate()
  n5 = ComponentMask(None=VectorParameter() {parameterName=Scale, defaultValue=<rgba 3.7337 5.3176 0 1>}.RGB) {g=true, r=true}
  n6 = VectorParameter() {parameterName=Offset, defaultValue=<rgba -0.1143 -0.4095 0 1>}
  n7 = MaterialFunctionCall!(
   Texture Scale=ScalarParameter() {defaultValue=1, parameterName=MinimizeAnim}
   UVs=Add(A=Multiply(A=@n4, B=@n5), B=ComponentMask(None=@n6.RGB) {g=true, r=true})
   Offset XY Texturesize Z=VectorParameter() {parameterName=PosScaling, defaultValue=<rgba 0.2567 0.9869 1 1>}.RGB
  ) {materialFunction=/Engine/Functions/Engine_MaterialFunctions01/Texturing/ScaleUVsAroundPoint}
  n8 = VectorParameter() {parameterName=RowsCollums, defaultValue=<rgba 7 7 0 1>}
  n9 = ScalarParameter() {defaultValue=7, parameterName=Speed}
  n10 = Time()
  n11 = Multiply(
   A=@n7.0-1 mask
   B=MaterialFunctionCall!(
    Texture=TextureObject() {texture=/BRCosmetics/Environments/FrontEnd/Textures/BP12/Play_FlipBook}
    UVs=@n7.UVs
    SubImages=ComponentMask(None=@n8.RGB) {g=true, r=true}
    Frame=Multiply(A=@n9, B=@n10)
   ) {materialFunction=/Engine/Functions/Engine_MaterialFunctions03/Texturing/SubUV_Function}.RGB
  )
  n12 = TextureSampleParameter2D() {parameterName=Normal, samplerType=SAMPLERTYPE_Normal, texture=/Game/Environments/Apollo/Sets/Agency/Props/Textures/T_Agency_Computer_01_N}
  n13 = Subtract(A=ObjectPositionWS(), B=WorldPosition().XYZ)
  n14 = Multiply(A=VertexNormalWS(), B=ScalarParameter() {defaultValue=0.4667, parameterName=VertOffset})
  n15 = MakeMaterialAttributes!(
   BaseColor=LinearInterpolate(
    A=TextureSampleParameter2D() {parameterName=Diffuse, texture=/Game/Environments/Apollo/Sets/Agency/Props/Textures/T_Agency_Computer_01_D}.RGB
    B=Constant3Vector()
    Alpha=@n2.R
   )
   Metallic=@n3.G
   Specular=@n3.G
   Roughness=@n3.G
   Anisotropy
   EmissiveColor=StaticSwitchParameter(True=Multiply(A=@n11, B=@n2.R), False=Constant()) {defaultValue=true, parameterName=UseFlipbook}
   Opacity
   OpacityMask
   Normal=@n12.RGB
   Tangent
   WorldPositionOffset=StaticSwitchParameter(True=LinearInterpolate(A=@n13, B=@n14, Alpha=@n2.R), False=Constant3Vector()) {defaultValue=true, parameterName=UseVertexshader}
   SubsurfaceColor
   ClearCoat
   ClearCoatRoughness
   AmbientOcclusion
   Refraction
   CustomizedUVs_0
   CustomizedUVs_1
   CustomizedUVs_2
   CustomizedUVs_3
   CustomizedUVs_4
   CustomizedUVs_5
   CustomizedUVs_6
   CustomizedUVs_7
   PixelDepthOffset
   ShadingModel
   Displacement
  )
  n16 = Panner(Coordinate=TextureCoordinate(), Speed=AppendVector(A=ScalarParameter() {defaultValue=-0.4381, parameterName=SpeedTreadmill}, B=Constant()))
  n17 = TextureSampleParameter2D(UVs=@n16) {parameterName=Spec, samplerType=SAMPLERTYPE_Masks, texture=&T_Agency_Computer_01_S}
  n18 = TextureSampleParameter2D(UVs=@n16) {parameterName=Normal, samplerType=SAMPLERTYPE_Normal, texture=&T_Agency_Computer_01_N}
  n19 = SetMaterialAttributes!(
   MaterialAttributes
   Base Color=TextureSampleParameter2D(UVs=@n16) {parameterName=Diffuse, texture=&T_Agency_Computer_01_D}.RGB
   Metallic=@n17.G
   Specular=@n17.G
   Roughness=@n17.G
   Normal=@n18.RGB
  ) {attributeSetTypes [
    #1
    #2
    #3
    #4
    #5
   ]}
  StaticSwitchParameter(True=BlendMaterialAttributes(A=@n15, B=@n19, Alpha=VertexColor().A) {vertexAttributeBlendType=UseA}, False=@n15) {parameterName=UseTreadmill}
 ]
}
