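# Unlit material for a progress bar in the BP12 front-end environment, drawn as a colored glowing band over black.
# The Progress scalar sets a hard U-axis cutoff through subtract and Ceil, filling from the U=1 edge.
# A parabola across V, raised to the Fallof power, fades the band toward both edges and scales the HDR Color parameter.
# A lerp uses the cutoff mask to choose between black and that band.
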Material /BRCosmetics/Environments/FrontEnd/Materials/BP12/M_Progressbar
graph {
 properties {pixelDepthOffsetMode=PDOM_Legacy, shadingModel=MSM_Unlit, usedShadingModels=Unlit}
 nodes [
  n1 = TextureCoordinate()
  n2 = Saturate(None=ScalarParameter() {defaultValue=0.281, parameterName=Progress})
  n3 = Saturate(None=Subtract(A=ComponentMask(None=@n1) {r=true}, B=OneMinus(None=@n2)))
  Floor(None=@n3)
  n4 = ComponentMask(None=TextureCoordinate()) {g=true}
  n5 = Multiply(A=Multiply(A=@n4, B=OneMinus(None=@n4))) {constB=2}
  n6 = ScalarParameter() {defaultValue=2.4385, parameterName=Fallof}
  LinearInterpolate(
   A=Constant3Vector()
   B=Multiply(A=Power(Base=@n5, Exp=@n6), B=VectorParameter() {parameterName=Color, defaultValue=<rgba 0 7.2909 30 1>}.RGB)
   Alpha=Ceil(None=@n3)
  )
  OneMinus(None=@n3)
 ]
}
