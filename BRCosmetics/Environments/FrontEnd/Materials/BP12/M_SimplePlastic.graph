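# M_SimplePlastic is a simple opaque Default Lit plastic material in the BP12 front-end environment folder, likely for dice given its parameter name.
# A DieColor vector parameter, dark red by default, provides an adjustable color.
# Three scalar constants presumably set fixed surface properties, though their output connections are not shown.
# It contains no textures or math nodes.
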Material /BRCosmetics/Environments/FrontEnd/Materials/BP12/M_SimplePlastic
graph {
 properties {pixelDepthOffsetMode=PDOM_Legacy, usedShadingModels=Default Lit}
 nodes [
  Constant()
  Constant() {r=1}
  Constant() {r=0.025}
  VectorParameter() {parameterName=DieColor, defaultValue=<rgba 0.5 0 0 1>}
 ]
}
