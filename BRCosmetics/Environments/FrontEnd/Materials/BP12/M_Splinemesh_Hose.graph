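# Spline-mesh hose material for the BP12 front end, with striped bands and an animated glow and bulge.
# Two sine waves along V form a hard band mask that blends RiftBeacon normal and mask textures toward flat values and switches between two colors.
# A time-scrolled sine drives emissive glow, shaped by a view-to-normal falloff, and a vertex-normal offset term for the bulge.
# Output pin connections are not shown, and some branches appear unused.
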Material /BRCosmetics/Environments/FrontEnd/Materials/BP12/M_Splinemesh_Hose
graph {
 properties {bUsedWithSplineMeshes=true, pixelDepthOffsetMode=PDOM_Legacy, usedShadingModels=Default Lit}
 nodes [
  n1 = Multiply(A=TextureCoordinate(), B=Constant2Vector() {g=1, r=1})
  n2 = ScalarParameter() {defaultValue=1, parameterName=Frequence2}
  n3 = ComponentMask(None=TextureCoordinate()) {g=true}
  n4 = Subtract(A=Sine(None=Multiply(A=@n2, B=@n3))) {constB=0.5}
  n5 = Multiply(A=ScalarParameter() {defaultValue=9, parameterName=Frequence}, B=@n3)
  n6 = Multiply(A=Ceil(None=Saturate(None=@n4)), B=Saturate(None=Sine(None=@n5)))
  n7 = Ceil(None=Saturate(None=@n6))
  LinearInterpolate(
   A=TextureSample(UVs=@n1) {samplerType=SAMPLERTYPE_Normal, texture=/Game/Environments/Props/RiftBeacon/Textures/T_RiftBeaconA_Bits_N}.RGB
   B=Constant3Vector() {constant=<rgba 0 0 1 0>}
   Alpha=@n7
  )
  n8 = VectorParameter() {parameterName=Color, defaultValue=<rgba 1 0 0 1>}
  LinearInterpolate(A=@n8.RGB, B=VectorParameter() {parameterName=Color2, defaultValue=<rgba 0 0 0 1>}.RGB, Alpha=@n7)
  n9 = TextureSample(UVs=@n1) {samplerType=SAMPLERTYPE_Masks, texture=/Game/Environments/Props/RiftBeacon/Textures/T_RiftBeaconA_Bits_S}
  LinearInterpolate(A=@n9.RGB, B=VectorParameter() {parameterName=SMR, defaultValue=<rgba 0.5 1 0.3 1>}.RGB, Alpha=@n7)
  MaterialFunctionCall(Float3=@n9.RGB) {materialFunction=&BreakOutFloat3Components}
  n10 = VectorParameter() {parameterName=EmissiveColor, defaultValue=<rgba 0.25 0.017 0.0021 1>}
  n11 = Time()
  n12 = ScalarParameter() {defaultValue=-1, parameterName=SpeedWPO}
  n13 = Multiply(A=@n11, B=@n12)
  n14 = Sine(None=Add(A=@n3, B=Frac(None=@n13)))
  n15 = Add(A=@n14)
  n16 = LinearInterpolate(A=ScalarParameter() {defaultValue=1, parameterName=Glow1}, B=ScalarParameter() {parameterName=Glow2}, Alpha=Divide(A=@n15))
  n17 = ScalarParameter() {defaultValue=3, parameterName=EXp}
  n18 = CameraVectorWS()
  n19 = VertexNormalWS()
  n20 = ScalarParameter() {defaultValue=3, parameterName=Fresnel}
  Multiply(A=Multiply(A=@n10.RGB, B=Power(Base=@n16, Exp=@n17)), B=Power(Base=DotProduct(A=@n18, B=@n19), Exp=@n20))
  Multiply(A=Multiply(A=Saturate(None=@n15), B=VertexNormalWS()), B=ScalarParameter() {defaultValue=1, parameterName=WPOoffset})
  Multiply(A=TextureSample(UVs=@n1) {texture=/Game/Environments/Props/RiftBeacon/Textures/T_RiftBeaconA_Bits_D}.RGB, B=@n8.RGB)
  Multiply(A=@n10.RGB, B=@n6)
  OneMinus(None=Fresnel(ExponentIn=@n20, Normal=@n19) {baseReflectFraction=0})
  n21 = Add(A=@n11, B=ScalarParameter() {parameterName=TimeOffset})
  n22 = Add(B=Frac(None=Multiply(A=@n21, B=@n12)))
  Saturate(None=Add(A=Sine(None=@n22)))
  Saturate(None=@n14)
  Time()
 ]
}
