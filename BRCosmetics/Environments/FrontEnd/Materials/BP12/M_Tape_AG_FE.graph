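# M_Tape_AG_FE is a Default Lit material in the BP12 front-end environment set, apparently for a tape prop, flagged for use with Nanite meshes.
# The graph is minimal, holding only a mid-grey color constant and a scalar constant left at its default.
# It has no textures or parameters, and the file shows no connections from either constant to material outputs.
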Material /BRCosmetics/Environments/FrontEnd/Materials/BP12/M_Tape_AG_FE
graph {
 properties {bUsedWithNanite=true, pixelDepthOffsetMode=PDOM_Legacy, usedShadingModels=Default Lit}
 nodes [
  Constant3Vector() {constant=<rgba 0.4896 0.4896 0.4896 0>}
  Constant()
 ]
}
