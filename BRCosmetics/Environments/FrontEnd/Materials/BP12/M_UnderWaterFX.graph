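# M_UnderWaterFX is a translucent unlit material for an underwater effect in the front-end environment.
# Vertex-offset math places the UV-mapped mesh relative to the camera and actor water height, adding sine waves; LimitAreaBox collapses it when the camera strays in X/Y.
# Color blends two tints by scene depth, plus flipbook caustics projected onto geometry behind it and faded by height.
# Opacity combines a height gradient with depth fade.
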Material /BRCosmetics/Environments/FrontEnd/Materials/BP12/M_UnderWaterFX
graph {
 properties {blendMode=BLEND_Translucent, pixelDepthOffsetMode=PDOM_Legacy, shadingModel=MSM_Unlit, usedShadingModels=Unlit}
 nodes [
  n1 = WorldPosition() {worldPositionShaderOffset=WPT_ExcludeAllShaderOffsets}
  n2 = ComponentMask(None=ActorPositionWS()) {b=true}
  n3 = Subtract(A=@n2, B=ComponentMask(None=CameraPositionWS()) {b=true})
  n4 = Clamp(None=@n3) {clampMode=CMODE_ClampMax, maxDefault=0}
  n5 = TextureCoordinate()
  n6 = ComponentMask(None=@n5) {g=true}
  n7 = Multiply(A=Clamp(None=@n3) {clampMode=CMODE_ClampMin, maxDefault=0}, B=OneMinus(None=@n6))
  n8 = Transform(None=VectorParameter() {parameterName=Direction, defaultValue=<rgba 0 1 0 1>}.RGB) {transformSourceType=TRANSFORMSOURCE_Camera}
  n9 = Constant()
  n10 = Transform(None=Constant3Vector() {constant=<rgba 0 0 1 0>}) {transformSourceType=TRANSFORMSOURCE_Camera}
  n11 = Multiply(A=@n10, B=Constant3Vector() {constant=<rgba 1 1 0 0>})
  n12 = MaterialFunctionCall(Vector=@n11) {materialFunction=&SafeNormalize}
  n13 = ScalarParameter() {defaultValue=12, parameterName=PushForward2}
  n14 = Constant3Vector() {constant=<rgba 0 0 -1 0>}
  n15 = Clamp(None=OneMinus(None=DotProduct(A=@n10, B=@n14))) {minDefault=0}
  n16 = Multiply(A=Normalize(VectorInput=@n11), B=LinearInterpolate(A=Divide(A=@n13, B=@n15), B=ScalarParameter() {parameterName=PushForward}, Alpha=@n6))
  n17 = Constant()
  n18 = ScalarParameter() {defaultValue=10666.3252, parameterName=Up}
  n19 = ScalarParameter() {defaultValue=12, parameterName=Forward3}
  n20 = Multiply(
   A=Transform(None=MaterialFunctionCall(X=@n17, Y=@n18, Z=@n19) {materialFunction=&MakeFloat3}.Result) {transformSourceType=TRANSFORMSOURCE_Camera}
   B=OneMinus(None=@n6)
  )
  n21 = Multiply(A=Subtract(A=ComponentMask(None=@n5) {r=true}) {constB=0.5}) {constB=2}
  n22 = ScalarParameter() {defaultValue=149.776, parameterName=LeftRight}
  n23 = ScalarParameter() {defaultValue=-52.858, parameterName=Down}
  n24 = ScalarParameter() {parameterName=Forward}
  n25 = ScalarParameter() {defaultValue=12, parameterName=Forward2}
  n26 = Transform(None=MaterialFunctionCall(X=Multiply(A=@n21, B=@n22), Y=Multiply(A=@n6, B=@n23), Z=LinearInterpolate(A=@n24, B=@n25, Alpha=@n6)) {materialFunction=&MakeFloat3}.Result) {transformSourceType=TRANSFORMSOURCE_Camera}
  n27 = ActorPositionWS()
  n28 = Subtract(A=@n27, B=WorldPosition().XYZ)
  n29 = Add(A=Subtract(A=CameraPositionWS(), B=@n27), B=@n28)
  n30 = Add(
   A=Multiply(A=Add(A=@n4, B=@n7), B=@n8)
   B=Add(A=If(A=@n9, B=@n12."Length==0", "A > B"=@n16, "A == B"=@n16, "A < B"=@n20), B=Add(A=@n26, B=@n29))
  )
  Add(A=@n1.XYZ, B=@n30)
  n31 = MaterialFunctionCall() {materialFunction=&WorldPositionBehindTranslucency}
  n32 = ScalarParameter() {defaultValue=440.8842, parameterName=SizeCaustics}
  n33 = Divide(A=Add(A=ComponentMask(None=@n31.Result) {g=true, r=true}, B=ComponentMask(None=@n31.Result) {b=true}), B=@n32)
  n34 = Time() {bOverride_Period=true, period=36}
  n35 = ScalarParameter() {defaultValue=27, parameterName=SpeedCaustics}
  n36 = Subtract(A=ComponentMask(None=@n31.Result) {b=true}, B=ComponentMask(None=ActorPositionWS()) {b=true})
  n37 = ScalarParameter() {defaultValue=-20.065, parameterName=FadeCaustics}
  n38 = LinearInterpolate(
   A=Constant3Vector()
   B=MaterialFunctionCall!(
    Texture=TextureObject() {samplerType=SAMPLERTYPE_LinearColor, texture=/FortWater/Textures/T_Siren_Caustics_03}
    UVs=Frac(None=@n33)
    SubImages=Constant2Vector() {g=6, r=6}
    Frame=Multiply(A=@n34, B=@n35)
   ) {materialFunction=/Engine/Functions/Engine_MaterialFunctions03/Texturing/SubUV_Function}.RGB
   Alpha=Saturate(None=Divide(A=@n36, B=@n37))
  )
  n39 = ScalarParameter() {defaultValue=0.1905, parameterName=CausticsBrightness}
  n40 = Saturate(None=Divide(A=SceneDepth().R, B=ScalarParameter() {defaultValue=1024, parameterName=Color Fade Length}))
  n41 = ScalarParameter() {defaultValue=1, parameterName=Color Fade Power}
  Add(
   A=Desaturation(None=Multiply(A=@n38, B=@n39), Fraction=ScalarParameter() {defaultValue=0.6381, parameterName=CausticsSat})
   B=LinearInterpolate(
    A=VectorParameter() {parameterName=Color2, defaultValue=<rgba 0.0834 0.416 0.599 1>}.RGB
    B=VectorParameter() {parameterName=Color, defaultValue=<rgba 0.0834 0.416 0.599 1>}.RGB
    Alpha=Power(Base=@n40, Exp=@n41)
   )
  )
  AppendVector()
  Divide(A=@n25)
  Divide(A=ScreenPosition().ViewportUV, B=@n32)
  DotProduct(A=@n31.Result, B=CameraVectorWS())
  MaterialFunctionCall() {materialFunction=&SubUV_Function}
  Multiply(A=@n28, B=Constant3Vector() {constant=<rgba 1 1 0 0>})
  n42 = Divide(A=Subtract(A=ComponentMask(None=@n1.XYZ) {b=true}, B=@n2), B=ScalarParameter() {defaultValue=-0.5987, parameterName=GradientDistance2})
  n43 = ScalarParameter() {defaultValue=-0.019, parameterName=Opacity}
  n44 = ScalarParameter() {defaultValue=0.5524, parameterName=Opacity2}
  n45 = DepthFade(Opacity=Constant() {r=1}, FadeDistance=ScalarParameter() {defaultValue=500, parameterName=Dpeth})
  Reroute(None=Multiply(A=Clamp(None=@n42), B=LinearInterpolate(A=@n43, B=@n44, Alpha=@n45)))
  n46 = Subtract(A=ComponentMask(None=ObjectPositionWS()) {g=true}, B=ComponentMask(None=CameraPositionWS()) {g=true})
  n47 = Floor(None=Subtract(A=Abs(None=@n46), B=ScalarParameter() {defaultValue=500, parameterName=FadeDistanceY}))
  Saturate(None=@n47)
  n48 = DotProduct(
   A=Normalize(VectorInput=VectorParameter() {parameterName=WaveDirection_A, defaultValue=<rgba 1 1 0 1>}.RGB)
   B=Add(A=WorldPosition() {worldPositionShaderOffset=WPT_ExcludeAllShaderOffsets}.XYZ, B=@n30)
  )
  n49 = Multiply(A=Time(), B=ScalarParameter() {defaultValue=42.3988, parameterName=WaveSpeed1})
  n50 = Add(A=ComponentMask(None=@n48) {r=true}, B=@n49)
  n51 = Sine(None=Divide(A=ComponentMask(None=@n50) {r=true}, B=ScalarParameter() {defaultValue=145.9291, parameterName=WaveLength1}))
  n52 = ComponentMask(None=@n48) {g=true}
  n53 = Sine(None=Divide(A=Add(A=@n52, B=@n49), B=ScalarParameter() {defaultValue=156.9975, parameterName=WaveLength2}))
  n54 = AppendVector(A=Constant2Vector(), B=Multiply(A=Add(A=@n51, B=@n53), B=ScalarParameter() {defaultValue=4, parameterName=Amplitude}))
  n55 = ScalarParameter() {defaultValue=3.9202, parameterName=VetexShaderWaveadd}
  n56 = Multiply(A=ScalarParameter() {defaultValue=0.2, parameterName=VetexShaderWave}, B=Multiply(A=Add(A=@n54, B=@n55), B=@n8))
  n57 = ObjectPositionWS()
  n58 = WorldPosition()
  n59 = ObjectPositionWS()
  n60 = CameraPositionWS()
  n61 = Abs(None=Subtract(A=ComponentMask(None=@n59) {r=true}, B=ComponentMask(None=@n60) {r=true}))
  n62 = ScalarParameter() {defaultValue=500, parameterName=FadeDistanceX}
  n63 = Add(A=Floor(None=Subtract(A=@n61, B=@n62)), B=@n47)
  StaticSwitchParameter(True=LinearInterpolate(A=Add(A=@n56, B=@n30), B=Subtract(A=@n57, B=@n58.XYZ), Alpha=Saturate(None=@n63)), False=Add(A=@n54, B=@n30)) {defaultValue=true, parameterName=LimitAreaBox}
  WorldPosition()
 ]
}
