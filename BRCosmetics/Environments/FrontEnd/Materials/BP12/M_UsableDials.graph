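# Lit material for a BP12 front-end prop with oil rig load bank textures and three independently rotatable dials.
# Vertex color R, G and B mask per-dial RotateAboutAxis offsets, each with its own angle and local-space pivot and a shared local axis parameter.
# The summed offsets add to a Shake-scaled sine wave over time and world position, weighted by distance from the actor origin and off by default.
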Material /BRCosmetics/Environments/FrontEnd/Materials/BP12/M_UsableDials
graph {
 properties {pixelDepthOffsetMode=PDOM_Legacy, usedShadingModels=Default Lit}
 nodes [
  n1 = Divide(A=Distance(A=ActorPositionWS(), B=WorldPosition().XYZ), B=ScalarParameter() {defaultValue=513.8858, parameterName=Falloff})
  n2 = ScalarParameter() {defaultValue=9.0026, parameterName=speed}
  n3 = Time()
  n4 = WorldPosition()
  n5 = ScalarParameter() {defaultValue=506.3531, parameterName=Frequency}
  n6 = Sine(None=Add(A=Multiply(A=@n2, B=@n3), B=Divide(A=@n4.XYZ, B=@n5)))
  n7 = VertexColor()
  n8 = VectorParameter() {parameterName=RoationAxis, defaultValue=<rgba 0 1 0 1>}
  n9 = RotateAboutAxis(
   NormalizedRotationAxis=Normalize(VectorInput=Transform(None=@n8.RGB) {transformSourceType=TRANSFORMSOURCE_Local})
   RotationAngle=ScalarParameter() {defaultValue=-0.0376, parameterName=Dial_1_Rotation}
   PivotPoint=TransformPosition(None=VectorParameter() {parameterName=Pivot1, defaultValue=<rgba -13 39 133 1>}.RGB) {transformType=TRANSFORMPOSSOURCE_World}
   Position=WorldPosition().XYZ
  )
  n10 = Transform(None=VectorParameter() {parameterName=RoationAxis, defaultValue=<rgba 0 1 0 1>}.RGB) {transformSourceType=TRANSFORMSOURCE_Local}
  n11 = VectorParameter() {parameterName=Pivot2, defaultValue=<rgba 2.2023 39 133.3 1>}
  n12 = Multiply(
   A=VertexColor().G
   B=RotateAboutAxis(
    NormalizedRotationAxis=Normalize(VectorInput=@n10)
    RotationAngle=ScalarParameter() {defaultValue=1.0945, parameterName=Dial_2_Rotation}
    PivotPoint=TransformPosition(None=@n11.RGB) {transformType=TRANSFORMPOSSOURCE_World}
    Position=WorldPosition().XYZ
   )
  )
  n13 = Transform(None=VectorParameter() {parameterName=RoationAxis, defaultValue=<rgba 0 1 0 1>}.RGB) {transformSourceType=TRANSFORMSOURCE_Local}
  n14 = VectorParameter() {parameterName=Pivot3, defaultValue=<rgba 17.4056 609.4121 133.2 1>}
  n15 = Multiply(
   A=VertexColor().B
   B=RotateAboutAxis(
    NormalizedRotationAxis=Normalize(VectorInput=@n13)
    RotationAngle=ScalarParameter() {defaultValue=-0.0567, parameterName=Dial_3_Rotation}
    PivotPoint=TransformPosition(None=@n14.RGB) {transformType=TRANSFORMPOSSOURCE_World}
    Position=WorldPosition().XYZ
   )
  )
  Add(A=Multiply(A=ScalarParameter() {parameterName=Shake}, B=Multiply(A=@n1, B=@n6)), B=Add(A=Multiply(A=@n7.R, B=@n9), B=Add(A=@n12, B=@n15)))
  TextureSample() {texture=/Game/Environments/Apollo/Sets/OilRig/Props/Textures/T_OilRig_LoadBank_01_D}
  TextureSample() {samplerType=SAMPLERTYPE_Masks, texture=/Game/Environments/Apollo/Sets/OilRig/Props/Textures/T_OilRig_LoadBank_01_S}
  TextureSample() {samplerType=SAMPLERTYPE_Normal, texture=/Game/Environments/Apollo/Sets/OilRig/Props/Textures/T_OilRig_LoadBank_01_N}
 ]
}
